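# Translucent UI material that builds an animated flame effect, named for a nitro state.
# Two noise textures pan at different tilings and rates, then are multiplied, raised to a power, boosted and saturated into a mask named "opacity".
# That mask drives a three-color blend between background, medium and bright color parameters.
# Both pan speeds are exposed as vector parameters.
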Material /MoonflaxMedallions/Materials/M_UI_NitroState_Flame
graph {
 properties {blendMode=BLEND_Translucent, materialDomain=MD_UI, pixelDepthOffsetMode=PDOM_Legacy, usedShadingModels=Default Lit}
 nodes [
  n1 = Time()
  n2 = Constant() {r=1}
  n3 = TextureSample(UVs=Panner(
    Coordinate=TextureCoordinate() {uTiling=1.5, vTiling=1.2}
    Time=Multiply(A=@n1, B=@n2)
    Speed=VectorParameter() {parameterName=Speed1, defaultValue=<rgba 0.2 0.15 0 1>}.RGB
   ) {speedX=0.3, speedY=0.8}) {
   samplerSource=SSM_Wrap_WorldGroupSettings
   samplerType=SAMPLERTYPE_LinearGrayscale
   texture=/Game/Characters/Player/Female/Medium/Bodies/F_MED_SuperNovaTaro/Textures/LowResBlurredNoise
  }
  n4 = Time()
  n5 = Constant() {r=0.6}
  n6 = TextureSample(UVs=Panner(
    Coordinate=TextureCoordinate() {uTiling=1.1, vTiling=0.9}
    Time=Multiply(A=@n4, B=@n5)
    Speed=VectorParameter() {parameterName=Speed2, defaultValue=<rgba 0.1 0.09 0 1>}.RGB
   ) {speedX=0.2, speedY=0.6}) {samplerSource=SSM_Wrap_WorldGroupSettings, texture=/Game/Athena/HUD/Phase/T_LowRes_Noise_Phase}
  n7 = Multiply(A=Power(Base=Multiply(A=@n3.RGB, B=@n6.RGB)), B=Constant() {r=10})
  MaterialFunctionCall(
   A=VectorParameter() {group=Color, parameterName=BackgroundColor, defaultValue=<rgba 0.3095 0.0123 0 1>}.RGB
   B=VectorParameter() {group=Color, parameterName=MediumColor, defaultValue=<rgba 1 0.1329 0 1>}.RGB
   C=VectorParameter() {group=Color, parameterName=BrightColor, defaultValue=<rgba 1 0.688 0.026 1>}.RGB
   Alpha=NamedRerouteDeclaration(None=Saturate(None=@n7)) {name=opacity, nodeColor=<rgba 0 0 0 1>}
  ) {materialFunction=&3ColorBlend}
 ]
}
