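# Moonflax equipped-item HUD widget showing ammo, resource or utility counts, adding an infinity display for the Nitro Queen medallion.
# Medallion count changes and Construct check the pawn's holding tag, swapping surplus ammo text for a flame-material infinity symbol.
# Two tagged slots host extension widgets in horizontal boxes.
# Magazine-text flame restyling nodes are unconnected, and resource count changes only print a debug string.
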WidgetBlueprint /MoonflaxMedallions/UI/WBP_MoonflaxEquippedItemWidget
graphs [
 AddContentWidget {nodes [
   n1 = AddContentWidget!() -> (then, WidgetSlotTag:Gameplay Tag Structure, ContentWidget:Widget Object Reference) {is=title K2Node_FunctionEntry}
   > n2 = Branch(Condition=MatchesTag!(
     TagOne:Gameplay Tag Structure=@n1
     TagTwo:Gameplay Tag Structure=Get ItemInfoSlotTag() -> (ItemInfoSlotTag:Gameplay Tag Structure)
     bExactMatch:bool=true
    ) -> (ReturnValue:bool))
   > n3 = AddChildToHorizontalBox!(
    execute
    self:Horizontal Box Object Reference=Get HBox_ExtensionWidgets() -> (HBox_ExtensionWidgets:Horizontal Box Object Reference)
    Content:Widget Object Reference=Get ContentWidget() -> (ContentWidget:Widget Object Reference)
   ) -> (then, ReturnValue:Horizontal Box Slot Object Reference)
   > SetHorizontalAlignment(self:Horizontal Box Slot Object Reference=@n3, InHorizontalAlignment=HAlign_Fill)
   > SetVerticalAlignment(self:Horizontal Box Slot Object Reference=@n3, InVerticalAlignment=VAlign_Fill)
   Branch(
    execute=@n2.else
    Condition=MatchesTag(
     TagOne=Get WidgetSlotTag() -> (WidgetSlotTag:Gameplay Tag Structure)
     TagTwo=Get ItemAmmInfoSlotTag() -> (ItemAmmInfoSlotTag:Gameplay Tag Structure)
     bExactMatch=true
    )
   )
   > n4 = AddChildToHorizontalBox(self=Get HBox_AmmoExtensionWidgets() -> (HBox_AmmoExtensionWidgets:Horizontal Box Object Reference), Content=Get ContentWidget())
   > SetHorizontalAlignment(self=@n4, InHorizontalAlignment=HAlign_Fill)
   > SetVerticalAlignment(self=@n4, InVerticalAlignment=VAlign_Fill)
  ]}
 EventGraph {nodes [
   HandleMedallionCountChanged!() -> (OutputDelegate, then, DidCountIncrease:bool=~, NewCount:int, NewestMedallionAdded:Fort Player Augment Item Definition Object Reference) {is=title K2Node_CustomEvent}
   > DelayUntilNextTick!(execute) -> (then)
   > Update Nitro Infinity State()
   Event Handle Resource Type Changed!() -> (OutputDelegate, then, Item:Fort Resource Item Definition Object Reference, ResourceCount:int=0) {is=title K2Node_Event}
   > Native_UpdateWidgetColor!(execute, self:Athena Equipped Item Widget Base Object Reference, bHasAmmo:bool=true) -> (then)
   n1 = Event BP_HandleHasBulletsPerClipChanged!() -> (OutputDelegate, then, bHasBulletsPerClip:bool) {is=title K2Node_Event}
   > SetVisibility(
    self=Get Image_AmmoSplit() -> (Image_AmmoSplit:Image Object Reference)
    InVisibility=Select(
     Option 0=Collapsed
     Option 1=Get WeaponEquipVisibility() -> (WeaponEquipVisibility:ESlateVisibility Enum=Visible)
     Index=@n1.bHasBulletsPerClip
    )
   )
   Event Destruct()
   > n2 = Cast To FortPlayerController(Object=GetOwningPlayer())
   > Unbind Event from On Medallion Count Changed(
    self:BP Medallions Tracking Component Object Reference=GetComponentByClass!(
     self:Actor Object Reference=@n2
     ComponentClass:Actor Component Class Reference=/Medallions/BP_Medallions_TrackingComponent.BP_Medallions_TrackingComponent_C
    ) -> (ReturnValue:BP Medallions Tracking Component Object Reference)
    Delegate=Create Event()
   ) {is=title K2Node_RemoveDelegate}
   n3 = Event Add Content To Host Widget!() -> (
    OutputDelegate
    then
    WidgetHostTag:"Gameplay Tag Structure (by ref)"
    ContentWidget:Widget Object Reference
    Padding:"Margin Structure (by ref)"
    bFill:bool
   ) {is=title K2Node_Event}
   > AddContentWidget!(execute, self, WidgetSlotTag:Gameplay Tag Structure=@n3.WidgetHostTag, ContentWidget:Widget Object Reference=@n3.ContentWidget) -> (then) {is=title K2Node_CallFunction}
   n4 = Event Handle Resource Count Changed!() -> (OutputDelegate, then, ResourceCount:int=0) {is=title K2Node_Event}
   > PrintString!(
    execute
    InString=Append(A=Resource Count:, B="To String (Integer)"(InInt=@n4.ResourceCount))
    bPrintToScreen
    bPrintToLog
    TextColor
    Duration
    Key
   ) -> (then)
   n5 = Event BP_HandleHideReserveAmmoChanged!() -> (OutputDelegate, then, bHideReserveAmmo:bool) {is=title K2Node_Event}
   > SetVisibility(
    self=[
     Get Text_SurplusUtilityAmmoCount() -> (Text_SurplusUtilityAmmoCount:Common Text Object Reference)
     Get Image_UtilitySplit() -> (Image_UtilitySplit:Image Object Reference)
     Get Text_SurplusAmmoCount() -> (Text_SurplusAmmoCount:Common Text Object Reference)
     Get Image_AmmoSplit()
    ]
    InVisibility=Select(Option 0=Get WeaponEquipVisibility() -> (WeaponEquipVisibility=Visible), Option 1=Collapsed, Index=@n5.bHideReserveAmmo)
   )
   Event Construct()
   > SetDesiredSizeOverride!(
    execute
    self:Image Object Reference=[
     Get Image_RechargeableIcon() -> (Image_RechargeableIcon:Image Object Reference)
     Get Image_ResourceIcon() -> (Image_ResourceIcon:Image Object Reference)
     Get Image_AmmoIcon() -> (Image_AmmoIcon:Image Object Reference)
    ]
    DesiredSize:Vector 2D Structure=Get SmallBrushSize() -> (SmallBrushSize:Vector 2D Structure)
   ) -> (then)
   > SetDesiredSizeOverride(
    self=Get Image_UtilityIcon() -> (Image_UtilityIcon:Image Object Reference)
    DesiredSize=Get LargeBrushSize() -> (LargeBrushSize:Vector 2D Structure)
   )
   > n6 = Cast To FortPlayerController(Object=GetOwningPlayer())
   > Bind Event to On Medallion Count Changed(
    self:BP Medallions Tracking Component Object Reference=GetComponentByClass(self=@n6, ComponentClass=&BP_Medallions_TrackingComponent_C)
    Delegate=Create Event()
   ) {is=title K2Node_AddDelegate}
   > Update Nitro Infinity State()
  ]}
 RegisterWidgetSlots {nodes [
   RegisterWidgetSlots() {is=title K2Node_FunctionEntry}
   > RegisterWidgetHost!(
    execute
    WidgetHost:Widget Object Reference
    WidgetHostTag:"Gameplay Tag Structure (by ref)"=Get ItemInfoSlotTag()
    IsPlayerSpecific:bool=true
   ) -> (then)
   > RegisterWidgetHost(WidgetHostTag=Get ItemAmmInfoSlotTag(), IsPlayerSpecific=true)
  ]}
 RemoveContentFromWidgetHost {nodes [
   n1 = Remove Content From Host Widget!() -> (
    then
    WidgetHostTag:"Gameplay Tag Structure (by ref)"
    ContentWidget:Widget Object Reference
    RemovalType:EWidgetSlottingRemovalType Enum=Unparent
   ) {is=title K2Node_FunctionEntry}
   > n2 = Branch(Condition=MatchesTag(TagOne=@n1, TagTwo=Get ItemInfoSlotTag(), bExactMatch=true))
   > n3 = RemoveChild(self=Get HBox_ExtensionWidgets(), Content=Get ContentWidget())
   > Return Node!(execute, ReturnValue:bool=@n3)
   n4 = Branch(
    execute=@n2.else
    Condition=MatchesTag(TagOne=Get WidgetHostTag() -> (WidgetHostTag:"Gameplay Tag Structure (by ref)"), TagTwo=Get ItemAmmInfoSlotTag(), bExactMatch=true)
   )
   > n5 = RemoveChild(self=Get HBox_AmmoExtensionWidgets(), Content=Get ContentWidget())
   > Return Node(ReturnValue=@n5)
   Return Node(execute=@n4.else, ReturnValue=false)
  ]}
 UnregisterWidgetSlots {nodes [
   UnregisterWidgetSlots() {is=title K2Node_FunctionEntry}
   > UnregisterWidgetHost!(execute, WidgetHost:Widget Object Reference, WidgetHostTag:"Gameplay Tag Structure (by ref)"=Get ItemInfoSlotTag()) -> (then)
   > UnregisterWidgetHost(WidgetHostTag=Get ItemAmmInfoSlotTag())
  ]}
 Update Nitro Infinity State {nodes [
   n1 = Get Text_MagazineAmmoCount() -> (Text_MagazineAmmoCount:Common Text Object Reference)
   SetFontMaterial!(execute, self:Text Object Reference=@n1, InMaterial:Material Interface Object Reference) -> (then)
   > Set Outline Color!(In:Exec, CommonTextWidget:Common Text Object Reference=@n1, Color:Linear Color Structure=(R=0,G=0,B=0,A=0.502)) -> (Out:Exec) {is=title K2Node_MacroInstance}
   n2 = Get Text_MagazineAmmoCount()
   SetFontMaterial(self=@n2, InMaterial=Get NitroMaterial() -> (NitroMaterial:Material Interface Object Reference))
   > Set Outline Color(CommonTextWidget=@n2, Color=(R=1,G=0.0319,B=0,A=1))
   Update Nitro Infinity State() {is=title K2Node_FunctionEntry}
   > n3 = HasMatchingGameplayTag!(
    execute:Exec
    self:Object Reference=Get Controlled Pawn!(self:Controller Object Reference=GetPlayerController!(PlayerIndex:int=0) -> (ReturnValue:Player Controller Object Reference)) -> (ReturnValue:Pawn Object Reference) {is=title K2Node_CallFunction}
    TagToCheck:Gameplay Tag Structure=(TagName="Granted.TakedownMedallion.IsHolding.Moonflax.NitroQueen")
   ) -> (then:Exec, ReturnValue:bool) {is=title K2Node_Message}
   > n4 = Branch(Condition=@n3)
   > SetVisibility(self=Get Text_NitroInfinity() -> (Text_NitroInfinity:Common Text Object Reference), InVisibility=Visible)
   > SetVisibility(self=Get Text_SurplusAmmoCount(), InVisibility=Collapsed)
   SetVisibility(execute=@n4.else, self=Get Text_NitroInfinity(), InVisibility=Collapsed)
   > SetVisibility(self=Get Text_SurplusAmmoCount(), InVisibility=HitTestInvisible)
  ]}
]
variables [
 ItemAmmInfoSlotTag {type=GameplayTag, default {tagName=DynamicUI.HUD.Fort.Athena.EquippedItemAmmoInfo.Slot}}
 ItemInfoSlotTag {type=GameplayTag, default {tagName=DynamicUI.HUD.Fort.Athena.EquippedItemInfo.Slot}}
 LargeBrushSize {type=Vector2D, default=<40 40>}
 NitroMaterial {type=MaterialInterface, default=/MoonflaxMedallions/Materials/M_UI_NitroState_Flame}
 SmallBrushSize {type=Vector2D, default=<32 32>}
]
widgets {
 info {
  inheritedWidgetCount=24
  namedSlotCount=1
  parentClass=/Script/FortniteUI.AthenaEquippedItemWidgetBase
  rootWidgetClass=/Script/FortniteUI.FortSimpleWidgetAnimationsPanel
  widgetCount=35
 }
 widgets [
  SimpleAnimations {
   widgetClassPath=&FortSimpleWidgetAnimationsPanel
   bInherited=true
   bIsVariable=true
   widget=SimpleAnimations
   properties {animations [
     NameIntro {
      bIsEnabled=true
      onAnimationsFinished=()
      targets [
       {
        alpha=/Game/Athena/HUD/AthenaEquippedItemWidget_Intro_ExtensionWidgets_Opacity
        bResetOnFinish=false
        bShouldAnimateAlpha=true
        bShouldAnimateScale=false
        bShouldAnimateTranslation=true
        scale=null
        targetName=HBox_ExtensionWidgets
        translation=/Game/Athena/HUD/AthenaEquippedItemWidget_Intro_ExtensionWidgets_Translation
        materialScalarAnimations []
       }
       {
        alpha=/Game/Athena/HUD/AthenaEquippedItemWidget_Intro_PopupText_Opacity
        bResetOnFinish=false
        bShouldAnimateAlpha=true
        bShouldAnimateScale=false
        bShouldAnimateTranslation=true
        scale=null
        targetName=SwitchPopupText
        translation=/Game/Athena/HUD/AthenaEquippedItemWidget_Intro_PopupText_Translation
        materialScalarAnimations []
       }
      ]
     }
     NameOutro {
      bIsEnabled=true
      onAnimationsFinished=()
      targets [
       {
        alpha=/Game/Athena/HUD/AthenaEquippedItemWidget_Outro_ExtensionWidgets_Opacity
        bResetOnFinish=false
        bShouldAnimateAlpha=true
        bShouldAnimateScale=false
        bShouldAnimateTranslation=true
        scale=null
        targetName=HBox_ExtensionWidgets
        translation=/Game/Athena/HUD/AthenaEquippedItemWidget_Outro_ExtensionWidgets_Translation
        materialScalarAnimations []
       }
       {
        alpha=/Game/Athena/HUD/AthenaEquippedItemWidget_Outro_PopupText_Opacity
        bResetOnFinish=false
        bShouldAnimateAlpha=true
        bShouldAnimateScale=false
        bShouldAnimateTranslation=true
        scale=null
        targetName=SwitchPopupText
        translation=/Game/Athena/HUD/AthenaEquippedItemWidget_Outro_PopupText_Translation
        materialScalarAnimations []
       }
      ]
     }
    ]}
  }
  SizeBox_1 {
   widgetClassPath=&SizeBox
   parent=SimpleAnimations
   widget=SizeBox_1
   properties {bOverride_HeightOverride=true, bOverride_WidthOverride=true, heightOverride=77, widthOverride=400}
   slot {class=FortSimpleWidgetAnimationsSlot}
  }
  VerticalBox_1 {widgetClassPath=&VerticalBox, parent=SizeBox_1, widget=VerticalBox_1, slot {class=SizeBoxSlot}}
  SizeBox_0 {
   widgetClassPath=&SizeBox
   parent=VerticalBox_1
   widget=SizeBox_0
   properties {bOverride_HeightOverride=true, heightOverride=48, minDesiredHeight=60}
   slot {class=VerticalBoxSlot}
  }
  HorizontalBox_0 {
   widgetClassPath=&HorizontalBox
   parent=SizeBox_0
   widget=HorizontalBox_0
   slot {class=SizeBoxSlot, properties {horizontalAlignment=HAlign_Right}}
  }
  FortHUDElementWrapper {
   widgetClassPath=/Game/Athena/HUD/FortHUDElementWrapper
   bIsVariable=true
   parent=HorizontalBox_0
   widget=FortHUDElementWrapper
   properties {
    hiddenVisibility=Hidden
    elementTag {tagName=HUD.Athena.EquippedAmmoIndicator}
    hUDElementTag {gameplayTags [{tagName=HUD.Athena.Hide.EquippedAmmoIndicator}]}
   }
   slot {class=HorizontalBoxSlot}
  }
  Overlay_28 {widgetClassPath=&Overlay, namedSlotHost=FortHUDElementWrapper, widget=Overlay_28}
  Switcher_Content {
   widgetClassPath=&CommonVisibilitySwitcher
   bInherited=true
   parent=Overlay_28
   widget=Switcher_Content
   properties {activeWidgetIndex=1}
   slot {class=OverlaySlot, properties {horizontalAlignment=HAlign_Fill, verticalAlignment=VAlign_Center}}
  }
  Panel_ResourceContent {
   widgetClassPath=&HorizontalBox
   bInherited=true
   parent=Switcher_Content
   widget=Panel_ResourceContent
   slot {class=CommonVisibilitySwitcherSlot, properties {horizontalAlignment=HAlign_Right, verticalAlignment=VAlign_Center, padding=<ltrb 0 0 4 0>}}
  }
  Image_ResourceIcon {
   widgetClassPath=&Image
   bInherited=true
   bIsVariable=true
   parent=Panel_ResourceContent
   widget=Image_ResourceIcon
   properties {
    visibility=SelfHitTestInvisible
    brush {
     bIsDynamicallyLoaded=false
     drawAs=Image
     imageType=NoImage
     mirroring=NoMirror
     resourceName=None
     resourceObject=/Game/UI/Foundation/Textures/Icons/Items/T-Icon-S-Loot-Stone
     tiling=NoTile
     imageSize=<40 40>
     margin=<ltrb 0 0 0 0>
     outlineSettings {
      bUseBrushTransparency=false
      roundingType=HalfHeightRadius
      width=0
      color {colorUseRule=UseColor_Specified, specifiedColor=<rgba 0 0 0 0>}
      cornerRadii=<0 0 0 0>
     }
     tintColor {colorUseRule=UseColor_Specified, specifiedColor=<rgba 1 1 1 1>}
     uVRegion {bIsValid=false, max=<0 0>, min=<0 0>}
    }
   }
   slot {class=HorizontalBoxSlot, properties {horizontalAlignment=HAlign_Right, verticalAlignment=VAlign_Center}}
  }
  Text_ResourceCount {
   widgetClassPath=&CommonTextBlock
   bInherited=true
   bIsVariable=true
   parent=Panel_ResourceContent
   widget=Text_ResourceCount
   properties {
    lineHeightPercentage=0.9
    text=999
    font {
     bForceMonospaced=false
     bMaterialIsStencil=false
     fontMaterial=null
     fontObject=&Burbank
     letterSpacing=0
     monospacedWidth=1
     size=24
     skewAmount=0
     typefaceFontName=BurbankRegular-Black
     outlineSettings {
      bApplyOutlineToDropShadows=true
      bMiteredCorners=false
      bSeparateFillAlpha=true
      outlineMaterial=&M-FontShadow
      outlineSize=3
      outlineColor=<rgba 0 0 0 0.5>
     }
    }
    margin=<ltrb 0 -2 0 0>
    renderTransform {angle=0, scale=<1 1>, shear=<-8 0>, translation=<0 0>}
    shadowOffset=<0 0>
   }
   slot {class=HorizontalBoxSlot, properties {horizontalAlignment=HAlign_Left, verticalAlignment=VAlign_Center, padding=<ltrb 4 0 0 0>}}
  }
  Panel_AmmoContent {
   widgetClassPath=&HorizontalBox
   bInherited=true
   parent=Switcher_Content
   widget=Panel_AmmoContent
   properties {renderTransform {angle=0, scale=<1 1>, shear=<-12 0>, translation=<0 0>}}
   slot {class=CommonVisibilitySwitcherSlot, properties {horizontalAlignment=HAlign_Right, verticalAlignment=VAlign_Center, padding=<ltrb 0 0 4 0>}}
  }
  Text_MagazineAmmoCount {
   widgetClassPath=&CommonTextBlock
   bInherited=true
   bIsVariable=true
   parent=Panel_AmmoContent
   widget=Text_MagazineAmmoCount
   properties {
    lineHeightPercentage=0.9
    text=0
    font {
     bForceMonospaced=false
     bMaterialIsStencil=false
     fontMaterial=null
     fontObject=&Burbank
     letterSpacing=0
     monospacedWidth=1
     size=28
     skewAmount=0
     typefaceFontName=Burbank-PlayerPower
     outlineSettings {
      bApplyOutlineToDropShadows=true
      bMiteredCorners=false
      bSeparateFillAlpha=true
      outlineMaterial=&M-FontShadow
      outlineSize=3
      outlineColor=<rgba 0 0 0 0.5>
     }
    }
    margin=<ltrb 0 -2 0 0>
    shadowOffset=<0 0>
   }
   slot {
    class=HorizontalBoxSlot
    properties {horizontalAlignment=HAlign_Right, verticalAlignment=VAlign_Center, padding=<ltrb 8 0 0 0>, size {sizeRule=Automatic, value=0.75}}
   }
  }
  Image_AmmoSplit {
   widgetClassPath=&Image
   bInherited=true
   bIsVariable=true
   parent=Panel_AmmoContent
   widget=Image_AmmoSplit
   properties {
    brush {
     bIsDynamicallyLoaded=false
     drawAs=Image
     imageType=NoImage
     mirroring=NoMirror
     resourceName=None
     resourceObject=/Game/Athena/HUD/Art/Materials/M_UI_HUD_AmmoCount_DividerLine
     tiling=NoTile
     imageSize=<4 22>
     margin=<ltrb 0 0 0 0>
     outlineSettings {
      bUseBrushTransparency=false
      roundingType=HalfHeightRadius
      width=0
      color {colorUseRule=UseColor_Specified, specifiedColor=<rgba 0 0 0 0>}
      cornerRadii=<0 0 0 0>
     }
     tintColor {colorUseRule=UseColor_Specified, specifiedColor=<rgba 1 1 1 1>}
     uVRegion {bIsValid=false, max=<0 0>, min=<0 0>}
    }
    renderTransform {angle=0, scale=<1 1>, shear=<0 0>, translation=<0 3>}
   }
   slot {class=HorizontalBoxSlot, properties {verticalAlignment=VAlign_Center, padding=<ltrb 4 0 4 0>}}
  }
  Text_SurplusAmmoCount {
   widgetClassPath=&CommonTextBlock
   bInherited=true
   bIsVariable=true
   parent=Panel_AmmoContent
   widget=Text_SurplusAmmoCount
   properties {
    text=0
    font {
     bForceMonospaced=false
     bMaterialIsStencil=false
     fontMaterial=null
     fontObject=&Burbank
     letterSpacing=0
     monospacedWidth=1
     size=20
     skewAmount=0
     typefaceFontName=Burbank-PlayerPower
     outlineSettings {
      bApplyOutlineToDropShadows=true
      bMiteredCorners=false
      bSeparateFillAlpha=true
      outlineMaterial=&M-FontShadow
      outlineSize=3
      outlineColor=<rgba 0 0 0 0.5>
     }
    }
    renderTransform {angle=0, scale=<1 1>, shear=<0 0>, translation=<0 3>}
    shadowOffset=<0 0>
   }
   slot {class=HorizontalBoxSlot, properties {horizontalAlignment=HAlign_Center, verticalAlignment=VAlign_Center}}
  }
  ScaleBox_Nitro {
   widgetClassPath=&ScaleBox
   parent=Panel_AmmoContent
   widget=ScaleBox_Nitro
   properties {stretch=ScaleBySafeZone, stretchDirection=UpOnly}
   slot {class=HorizontalBoxSlot, properties {verticalAlignment=VAlign_Center, padding=<ltrb 4 0 4 10>}}
  }
  Text_NitroInfinity {
   widgetClassPath=&CommonTextBlock
   bIsVariable=true
   parent=ScaleBox_Nitro
   widget=Text_NitroInfinity
   properties {
    text="∞"
    visibility=Collapsed
    font {
     bForceMonospaced=false
     bMaterialIsStencil=false
     fontMaterial=&M_UI_NitroState_Flame
     fontObject=&Burbank
     letterSpacing=0
     monospacedWidth=1
     size=60
     skewAmount=0
     typefaceFontName=Burbank-PlayerPower
     outlineSettings {
      bApplyOutlineToDropShadows=true
      bMiteredCorners=false
      bSeparateFillAlpha=false
      outlineMaterial=null
      outlineSize=1
      outlineColor=<rgba 1 0.2658 0 1>
     }
    }
    renderTransform {angle=0, scale=<1 1>, shear=<0 0>, translation=<0 3>}
    shadowOffset=<0 0>
   }
   slot {class=ScaleBoxSlot}
  }
  Image_AmmoIcon {
   widgetClassPath=&Image
   bInherited=true
   bIsVariable=true
   parent=Panel_AmmoContent
   widget=Image_AmmoIcon
   properties {
    visibility=SelfHitTestInvisible
    brush {
     bIsDynamicallyLoaded=false
     drawAs=Image
     imageType=NoImage
     mirroring=NoMirror
     resourceName=None
     resourceObject=/Game/UI/Foundation/Textures/Icons/Ammo/T-Icon-Ammo-Medium-64
     tiling=NoTile
     imageSize=<32 32>
     margin=<ltrb 0 0 0 0>
     outlineSettings {
      bUseBrushTransparency=false
      roundingType=HalfHeightRadius
      width=0
      color {colorUseRule=UseColor_Specified, specifiedColor=<rgba 0 0 0 0>}
      cornerRadii=<0 0 0 0>
     }
     tintColor {colorUseRule=UseColor_Specified, specifiedColor=<rgba 1 1 1 1>}
     uVRegion {bIsValid=false, max=<0 0>, min=<0 0>}
    }
    renderTransform {angle=0, scale=<1 1>, shear=<4 0>, translation=<0 0>}
   }
   slot {class=HorizontalBoxSlot, properties {horizontalAlignment=HAlign_Left, verticalAlignment=VAlign_Center, padding=<ltrb 4 0 0 0>}}
  }
  Panel_UtilityContent {
   widgetClassPath=&HorizontalBox
   bInherited=true
   parent=Switcher_Content
   widget=Panel_UtilityContent
   slot {class=CommonVisibilitySwitcherSlot, properties {horizontalAlignment=HAlign_Right, verticalAlignment=VAlign_Center, padding=<ltrb 0 0 4 0>}}
  }
  Image_UtilityIcon {
   widgetClassPath=&Image
   bInherited=true
   bIsVariable=true
   parent=Panel_UtilityContent
   widget=Image_UtilityIcon
   properties {
    visibility=SelfHitTestInvisible
    brush {
     bIsDynamicallyLoaded=false
     drawAs=Image
     imageType=NoImage
     mirroring=NoMirror
     resourceName=None
     resourceObject=&T-Icon-S-Loot-Stone
     tiling=NoTile
     imageSize=<40 40>
     margin=<ltrb 0 0 0 0>
     outlineSettings {
      bUseBrushTransparency=false
      roundingType=HalfHeightRadius
      width=0
      color {colorUseRule=UseColor_Specified, specifiedColor=<rgba 0 0 0 0>}
      cornerRadii=<0 0 0 0>
     }
     tintColor {colorUseRule=UseColor_Specified, specifiedColor=<rgba 1 1 1 1>}
     uVRegion {bIsValid=false, max=<0 0>, min=<0 0>}
    }
   }
   slot {class=HorizontalBoxSlot, properties {horizontalAlignment=HAlign_Right, verticalAlignment=VAlign_Center}}
  }
  Text_UtilityAmmoCount {
   widgetClassPath=&CommonTextBlock
   bInherited=true
   bIsVariable=true
   parent=Panel_UtilityContent
   widget=Text_UtilityAmmoCount
   properties {
    text=0%
    font {
     bForceMonospaced=false
     bMaterialIsStencil=false
     fontMaterial=null
     fontObject=&Burbank
     letterSpacing=0
     monospacedWidth=1
     size=28
     skewAmount=0.2
     typefaceFontName=Burbank-PlayerPower
     outlineSettings {
      bApplyOutlineToDropShadows=true
      bMiteredCorners=false
      bSeparateFillAlpha=true
      outlineMaterial=&M-FontShadow
      outlineSize=3
      outlineColor=<rgba 0 0 0 0.5>
     }
    }
    shadowOffset=<0 0>
   }
   slot {
    class=HorizontalBoxSlot
    properties {horizontalAlignment=HAlign_Right, verticalAlignment=VAlign_Center, padding=<ltrb 4 0 0 0>, size {sizeRule=Automatic, value=0.75}}
   }
  }
  Image_UtilitySplit {
   widgetClassPath=&Image
   bInherited=true
   bIsVariable=true
   parent=Panel_UtilityContent
   widget=Image_UtilitySplit
   properties {
    visibility=HitTestInvisible
    brush {
     bIsDynamicallyLoaded=false
     drawAs=Image
     imageType=NoImage
     mirroring=NoMirror
     resourceName=None
     resourceObject=&M_UI_HUD_AmmoCount_DividerLine
     tiling=NoTile
     imageSize=<4 22>
     margin=<ltrb 0 0 0 0>
     outlineSettings {
      bUseBrushTransparency=false
      roundingType=HalfHeightRadius
      width=0
      color {colorUseRule=UseColor_Specified, specifiedColor=<rgba 0 0 0 0>}
      cornerRadii=<0 0 0 0>
     }
     tintColor {colorUseRule=UseColor_Specified, specifiedColor=<rgba 1 1 1 1>}
     uVRegion {bIsValid=false, max=<0 0>, min=<0 0>}
    }
    renderTransform {angle=0, scale=<1 1>, shear=<-12 0>, translation=<0 3>}
   }
   slot {class=HorizontalBoxSlot, properties {verticalAlignment=VAlign_Center, padding=<ltrb 4 0 4 0>}}
  }
  Text_SurplusUtilityAmmoCount {
   widgetClassPath=&CommonTextBlock
   bInherited=true
   bIsVariable=true
   parent=Panel_UtilityContent
   widget=Text_SurplusUtilityAmmoCount
   properties {
    text=0
    visibility=HitTestInvisible
    font {
     bForceMonospaced=false
     bMaterialIsStencil=false
     fontMaterial=null
     fontObject=&Burbank
     letterSpacing=0
     monospacedWidth=1
     size=20
     skewAmount=0
     typefaceFontName=Burbank-PlayerPower
     outlineSettings {
      bApplyOutlineToDropShadows=true
      bMiteredCorners=false
      bSeparateFillAlpha=true
      outlineMaterial=&M-FontShadow
      outlineSize=3
      outlineColor=<rgba 0 0 0 0.5>
     }
    }
    renderTransform {angle=0, scale=<1 1>, shear=<-12 0>, translation=<0 3>}
    shadowOffset=<0 0>
   }
   slot {class=HorizontalBoxSlot, properties {horizontalAlignment=HAlign_Left, verticalAlignment=VAlign_Center}}
  }
  Text_OutOfAmmo {
   widgetClassPath=&CommonTextBlock
   bInherited=true
   bIsVariable=true
   parent=Overlay_28
   widget=Text_OutOfAmmo
   properties {
    text=NO AMMO
    visibility=Collapsed
    font {
     bForceMonospaced=false
     bMaterialIsStencil=false
     fontMaterial=null
     fontObject=&Burbank
     letterSpacing=0
     monospacedWidth=1
     size=18
     skewAmount=0
     typefaceFontName=BurbankRegular-Black
     outlineSettings {
      bApplyOutlineToDropShadows=true
      bMiteredCorners=false
      bSeparateFillAlpha=false
      outlineMaterial=null
      outlineSize=1
      outlineColor=<rgba 0 0 0 0.498>
     }
    }
    shadowOffset=<0 0>
   }
   slot {class=OverlaySlot, properties {horizontalAlignment=HAlign_Center, verticalAlignment=VAlign_Center}}
  }
  TutorialHighlight {
   widgetClassPath=&TutorialHighlight
   bIsVariable=true
   parent=Overlay_28
   widget=TutorialHighlight
   properties {
    customHighlightWidget=/Game/Athena/HUD/Tutorial/Tutorial_Outline_Square_Static.Tutorial_Outline_Square_Static_C
    visibility=SelfHitTestInvisible
   }
   slot {class=OverlaySlot, properties {horizontalAlignment=HAlign_Fill, verticalAlignment=VAlign_Fill}}
  }
  HBox_AmmoExtensionWidgets {
   widgetClassPath=&HorizontalBox
   bIsVariable=true
   parent=HorizontalBox_0
   widget=HBox_AmmoExtensionWidgets
   slot {class=HorizontalBoxSlot, properties {horizontalAlignment=HAlign_Right, padding=<ltrb 4 0 0 0>}}
  }
  Image_RechargeableIcon {
   widgetClassPath=&FortLazyImage
   bInherited=true
   bIsVariable=true
   parent=HBox_AmmoExtensionWidgets
   widget=Image_RechargeableIcon
   properties {
    bIsVolatile=true
    visibility=Collapsed
    brush {
     bIsDynamicallyLoaded=false
     drawAs=Image
     imageType=NoImage
     mirroring=NoMirror
     resourceName=None
     resourceObject=/Game/UI/Foundation/Textures/Icons/Ammo/T-Icon-RechargeAmmo-64
     tiling=NoTile
     imageSize=<32 32>
     margin=<ltrb 0 0 0 0>
     outlineSettings {
      bUseBrushTransparency=false
      roundingType=HalfHeightRadius
      width=0
      color {colorUseRule=UseColor_Specified, specifiedColor=<rgba 0 0 0 0>}
      cornerRadii=<0 0 0 0>
     }
     tintColor {colorUseRule=UseColor_Specified, specifiedColor=<rgba 1 1 1 1>}
     uVRegion {bIsValid=false, max=<0 0>, min=<0 0>}
    }
    loadingBackgroundBrush {
     bIsDynamicallyLoaded=false
     drawAs=NoDrawType
     imageType=NoImage
     mirroring=NoMirror
     resourceName=None
     resourceObject=null
     tiling=NoTile
     imageSize=<48 48>
     margin=<ltrb 0 0 0 0>
     outlineSettings {
      bUseBrushTransparency=false
      roundingType=HalfHeightRadius
      width=0
      color {colorUseRule=UseColor_Specified, specifiedColor=<rgba 0 0 0 0>}
      cornerRadii=<0 0 0 0>
     }
     tintColor {colorUseRule=UseColor_Specified, specifiedColor=<rgba 1 1 1 1>}
     uVRegion {bIsValid=false, max=<0 0>, min=<0 0>}
    }
   }
   slot {class=HorizontalBoxSlot, properties {horizontalAlignment=HAlign_Left, verticalAlignment=VAlign_Center, padding=<ltrb 4 0 4 0>}}
  }
  SwitchPopup {
   widgetClassPath=&VerticalBox
   bInherited=true
   bIsVariable=true
   parent=VerticalBox_1
   widget=SwitchPopup
   properties {visibility=Hidden}
   slot {class=VerticalBoxSlot, properties {verticalAlignment=VAlign_Top}}
  }
  Overlay_2 {
   widgetClassPath=&Overlay
   parent=SwitchPopup
   widget=Overlay_2
   properties {clipping=ClipToBounds}
   slot {class=VerticalBoxSlot, properties {verticalAlignment=VAlign_Center}}
  }
  SwitchPopupText {
   widgetClassPath=&CommonTextBlock
   bInherited=true
   bIsVariable=true
   parent=Overlay_2
   widget=SwitchPopupText
   properties {
    autoWrapText=true
    justification=Right
    text="[PH] Item Name"
    textTransformPolicy=ToUpper
    font {
     bForceMonospaced=false
     bMaterialIsStencil=false
     fontMaterial=null
     fontObject=&Burbank
     letterSpacing=0
     monospacedWidth=1
     size=15
     skewAmount=0
     typefaceFontName=BurbankRegular-Black
     outlineSettings {
      bApplyOutlineToDropShadows=true
      bMiteredCorners=false
      bSeparateFillAlpha=true
      outlineMaterial=&M-FontShadow
      outlineSize=2
      outlineColor=<rgba 0 0 0 0.5>
     }
    }
    shadowOffset=<0 0>
    shapedTextOptions {bOverride_TextFlowDirection=true, bOverride_TextShapingMethod=false, textFlowDirection=LeftToRight, textShapingMethod=Auto}
   }
   slot {class=OverlaySlot, properties {horizontalAlignment=HAlign_Fill, verticalAlignment=VAlign_Fill}}
  }
  HBox_ExtensionWidgets {
   widgetClassPath=&HorizontalBox
   bIsVariable=true
   parent=SwitchPopup
   widget=HBox_ExtensionWidgets
   slot {class=VerticalBoxSlot, properties {horizontalAlignment=HAlign_Right, padding=<ltrb 4 0 0 0>}}
  }
  Overlay_UtilityAmmoIcon {widgetClassPath=&Overlay, bInherited=true, widget=null}
  RootScaleBox {widgetClassPath=&ScaleBox, bInherited=true, widget=null}
  Image_Background {widgetClassPath=&Image, bInherited=true, widget=null}
  ScaleBox_HUDWidgetScale {widgetClassPath=&ScaleBox, bInherited=true, widget=null}
 ]
}
